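; Quest bullseye landing-target actor that shows a particle marker while the local player skydives, for an active quest objective.
; On remote clients, BeginPlay waits through a workaround delay, checks the quest and objective (or a Testing flag), then binds skydiving and quest-update events.
; Skydive begin and end toggle the particles.
; When a quest update finds the objective no longer valid, it spawns a landing effect, hides the actor and unbinds.
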(event Custom|COMEBACK
  ;[1 1 1] hooks for activating/deactivating bulls eye particle system
  (Components|Activation|Activate (Variables|Default|GetP_Athena_BullsEye) true))

(event Custom|GOAWAY
  ;[1 1 1] hooks for activating/deactivating bulls eye particle system
  (Components|Activation|Deactivate (Variables|Default|GetP_Athena_BullsEye)))

(event Custom|QuestUpdated
  (bind _bvalidobjective (CallFunction|IsObjectiveValid))
  (if (not _bvalidobjective)
    (CallFunction|LANDSUCCESS)))

(event Custom|LANDSUCCESS
  ;[1 1 1] show feedback for the player that they did the thing..
  (Effects|SpawnEmitteratLocation "/Game/Athena/Items/QuestInteractables/BullsEye/P_Athena_BullsEye_Land.P_Athena_BullsEye_Land" (Transformation|GetActorLocation))
  ;[1 1 1] quest is done - stop listening and clean up state
  (Rendering|SetActorHiddenInGame true)
  (bind _asathena_player_controller (Utilities|Casting|CastToAthena_PlayerController (Utilities|Array|Get(acopy) (Fortnite|GetLocalFortPlayerControllers true))))
  (EventDispatchers|UnbindEventfromOnEndSkydiving _asathena_player_controller (EventDispatchers|CreateEvent :!event Custom|GOAWAY))
  (EventDispatchers|UnbindEventfromOnEndSkydiving _asathena_player_controller (EventDispatchers|CreateEvent :!event Custom|COMEBACK))
  (Quests|UnbindEventfromOnQuestsUpdated (Quests|GetQuestManager _asathena_player_controller "Invalid") (EventDispatchers|CreateEvent :!event Custom|QuestUpdated)))

(event EventBeginPlay
  (bind _outputdelegate (EventDispatchers|CreateEvent :!event Custom|GOAWAY))
  (bind _outputdelegate_2 (EventDispatchers|CreateEvent :!event Custom|COMEBACK))
  (bind _outputdelegate_3 (EventDispatchers|CreateEvent :!event Custom|QuestUpdated))
  ;[1 1 1] have we done this quest before?
  (Networking|SwitchHasAuthority
    (:Remote
      ;;[1 0.5445 0.1188] Workaround - needs to be something better for ship
      (Utilities|FlowControl|Delay 5.0)
      (bind _questvalid (CallFunction|DoesPlayerHaveThisQuest))
      (if (or _questvalid (Variables|Default|GetTesting))
        (bind _bvalidobjective (CallFunction|IsObjectiveValid))
        (if (or _bvalidobjective (Variables|Default|GetTesting))
          (bind _asfort_player_controller_athena (Utilities|Casting|CastToFortPlayerControllerAthena (Utilities|Array|Get(acopy) (Fortnite|GetLocalFortPlayerControllers true))))
          (EventDispatchers|BindEventtoOnEndSkydiving _asfort_player_controller_athena _outputdelegate)
          (EventDispatchers|AssignOnBeginSkydiving _asfort_player_controller_athena _outputdelegate_2)
          (Quests|BindEventtoOnQuestsUpdated (Quests|GetQuestManager _asfort_player_controller_athena "Invalid") _outputdelegate_3)
          (else
            _)))
      ;[1 1 1] quest is done - stop listening and clean up state
      (Rendering|SetActorHiddenInGame true)
      (bind _asathena_player_controller (Utilities|Casting|CastToAthena_PlayerController (Utilities|Array|Get(acopy) (Fortnite|GetLocalFortPlayerControllers true))))
      (EventDispatchers|UnbindEventfromOnEndSkydiving _asathena_player_controller _outputdelegate)
      (EventDispatchers|UnbindEventfromOnEndSkydiving _asathena_player_controller _outputdelegate_2)
      (Quests|UnbindEventfromOnQuestsUpdated (Quests|GetQuestManager _asathena_player_controller "Invalid") _outputdelegate_3))))

(fn DoesPlayerHaveThisQuest
  (bind _returnvalue (Fortnite|GetLocalFortPlayerControllers true))
  (bind _output (Utilities|Array|Get(acopy) _returnvalue))
  (bind _returnvalue_2 (Quests|GetQuestManager _output "Invalid"))
  (bind _questitem (Variables|Default|GetQuestItem))
  (bind _returnvalue_3 (Quests|GetQuestwithDefinition _returnvalue_2 _questitem))
  (Utilities|IsValid _returnvalue_3
    (:Is%20Valid
      (LocalVariables|SetQuestIsValid (Utilities|Enum|Equal(Enum) (Quest|GetFortQuestState _returnvalue_3) "Active")))
    (:Is%20Not%20Valid
      (LocalVariables|SetQuestIsValid false)))
  (return (LocalVariables|GetQuestIsValid)))

(fn IsObjectiveValid
  (bind _questvalid (Quest|DoesPlayerHaveQuest (Utilities|Array|Get(acopy) (Fortnite|GetLocalFortPlayerControllers true)) (Variables|Default|GetQuestItem) (Variables|Default|GetQuestBackendObjectiveName)))
  (return _questvalid))

(fn ConstructionScript)
