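# Translucent unlit particle material for the BullsEye quest interactable, tinted by particle color with no textures and visible only from a distance.
# Pixel depth drives a clamped ramp that brightens the color from a dimmed fraction up to full as camera distance grows.
# A second depth ramp fades opacity out up close, multiplied by a DepthFade of particle alpha that softens intersections with scene geometry.
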Material /Figment_S10/Athena/Items/QuestInteractables/BullsEye/M_Athena_BullsEye
graph {
 properties {bUsedWithMeshParticles=true, blendMode=BLEND_Translucent, pixelDepthOffsetMode=PDOM_Legacy, shadingModel=MSM_Unlit, usedShadingModels=Unlit}
 nodes [
  n1 = ParticleColor()
  n2 = Subtract(A=PixelDepth().R, B=Constant() {r=500})
  n3 = Constant() {r=4000}
  LinearInterpolate(A=Multiply(A=@n1.RGB) {constB=0.2}, B=@n1.RGB, Alpha=Clamp(None=Divide(A=@n2, B=@n3)))
  n4 = Subtract(A=PixelDepth().R, B=Constant() {r=1000})
  n5 = Constant() {r=2500}
  Multiply(A=Clamp(None=Divide(A=@n4, B=@n5)), B=DepthFade(Opacity=@n1.A) {fadeDistanceDefault=200})
 ]
}
